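# Physics asset for the female medium Rebirth Soldier jacket, defining bodies and joints for its dynamic panels, zipper and collar clasp.
# Thin boxes form front, side and back panels, chained from small master capsules on spine_03, with a zipper ending the left-front chain.
# Master joints are angularly free, while panel joints restrict twist tightly and back panels swing less.
# Head and spine capsules represent the body, and the clasp hinges from spine_05 on one swing axis.
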PhysicsAsset /Game/Characters/Player/Female/Medium/Bodies/F_MED_Rebirth_Soldier/Meshes/F_MED_Rebirth_Soldier_Jacket
bodies {
 dyn_collar_clasp [{shapeType=Capsule, length=2, radius=0.7162, center=<1.1193 0.0001 0.0002>, rotation=<rot 90 0 0>}]
 dyn_jacket_bk_01 [{shapeType=Box, extentX=7.5, extentY=7.5, extentZ=0.6903, center=<3.9836 0.0001 0>}]
 dyn_jacket_bk_MSTR [{shapeType=Capsule, length=0.5, radius=0.5, rotation=<rot 90 0 0>}]
 dyn_jacket_bk_l_01 [{shapeType=Box, extentX=7.5, extentY=7.5, extentZ=0.6903, center=<4.0261 0.0001 0>}]
 dyn_jacket_bk_l_MSTR [{shapeType=Capsule, length=0.5, radius=0.5, rotation=<rot 90 0 0>}]
 dyn_jacket_bk_r_01 [{shapeType=Box, extentX=7.5, extentY=7.5, extentZ=0.6903, center=<-4.124 0 0>}]
 dyn_jacket_bk_r_MSTR [{shapeType=Capsule, length=0.5, radius=0.5, rotation=<rot 90 0 0>}]
 dyn_jacket_fr_l_01 [{shapeType=Box, extentX=5.6362, extentY=7.5, extentZ=0.6903, center=<2.6448 0.0001 0>}]
 dyn_jacket_fr_l_02 [{shapeType=Box, extentX=4.3562, extentY=7.5, extentZ=0.6903, center=<2.0136 0.0001 0>}]
 dyn_jacket_fr_l_MSTR [{shapeType=Capsule, length=0.5, radius=0.5, rotation=<rot 90 0 0>}]
 dyn_jacket_fr_l_zipper [{shapeType=Box, extentX=3.0392, extentY=1.6811, extentZ=0.6903, center=<1.4606 0.0001 -0.0001>}]
 dyn_jacket_fr_r_01 [{shapeType=Box, extentX=5.6752, extentY=7.5, extentZ=1.1903, center=<-2.5092 0 0>}]
 dyn_jacket_fr_r_02 [{shapeType=Box, extentX=3.4334, extentY=7.5, extentZ=1.1903, center=<-1.7355 0 0>}]
 dyn_jacket_fr_r_MSTR [{shapeType=Capsule, length=0.5, radius=0.5, rotation=<rot 90 0 0>}]
 dyn_jacket_l_01 [{shapeType=Box, extentX=7.5, extentY=7.5, extentZ=0.6903, center=<3.6782 0.0001 0>}]
 dyn_jacket_l_MSTR [{shapeType=Capsule, length=0.5, radius=0.5, rotation=<rot 90 0 0>}]
 dyn_jacket_r_01 [{shapeType=Box, extentX=7.5, extentY=7.5, extentZ=1.1903, center=<-3.6782 0 0>}]
 dyn_jacket_r_MSTR [{shapeType=Capsule, length=0.5, radius=0.5, rotation=<rot 90 0 0>}]
 head [{shapeType=Capsule, length=2.1922, radius=6.2541, center=<1.933 -3.591 0>, rotation=<rot 90 0 0>}]
 spine_03 [{shapeType=Capsule, length=2.9926, radius=8.4672, center=<-2.0188 2.4863 0>}]
 spine_05 [{shapeType=Capsule, length=24.1045, radius=4.3947, center=<5.254 -0.0001 0>}]
}
constraints [
 {
  bone1Name=dyn_jacket_fr_l_MSTR
  bone2Name=spine_03
  swing1LimitDegrees=45
  swing1Motion=Free
  swing2LimitDegrees=45
  swing2Motion=Free
  twistLimitDegrees=45
  twistMotion=Free
 }
 {bone1Name=dyn_jacket_fr_l_01, bone2Name=dyn_jacket_fr_l_MSTR, swing1LimitDegrees=45, swing2LimitDegrees=45, twistLimitDegrees=5}
 {bone1Name=dyn_jacket_fr_l_02, bone2Name=dyn_jacket_fr_l_01, swing1LimitDegrees=45, swing2LimitDegrees=45, twistLimitDegrees=5}
 {bone1Name=dyn_jacket_fr_l_zipper, bone2Name=dyn_jacket_fr_l_02, swing1LimitDegrees=45, swing2LimitDegrees=45, twistLimitDegrees=5}
 {
  bone1Name=dyn_jacket_l_MSTR
  bone2Name=spine_03
  swing1LimitDegrees=45
  swing1Motion=Free
  swing2LimitDegrees=45
  swing2Motion=Free
  twistLimitDegrees=45
  twistMotion=Free
 }
 {bone1Name=dyn_jacket_l_01, bone2Name=dyn_jacket_l_MSTR, swing1LimitDegrees=45, swing2LimitDegrees=45, twistLimitDegrees=5}
 {
  bone1Name=dyn_jacket_bk_l_MSTR
  bone2Name=spine_03
  swing1LimitDegrees=45
  swing1Motion=Free
  swing2LimitDegrees=45
  swing2Motion=Free
  twistLimitDegrees=45
  twistMotion=Free
 }
 {bone1Name=dyn_jacket_bk_l_01, bone2Name=dyn_jacket_bk_l_MSTR, swing1LimitDegrees=20, swing2LimitDegrees=30, twistLimitDegrees=5}
 {
  bone1Name=dyn_jacket_bk_MSTR
  bone2Name=spine_03
  swing1LimitDegrees=45
  swing1Motion=Free
  swing2LimitDegrees=45
  swing2Motion=Free
  twistLimitDegrees=45
  twistMotion=Free
 }
 {bone1Name=dyn_jacket_bk_01, bone2Name=dyn_jacket_bk_MSTR, swing1LimitDegrees=25, swing2LimitDegrees=30, twistLimitDegrees=5}
 {
  bone1Name=dyn_jacket_bk_r_MSTR
  bone2Name=spine_03
  swing1LimitDegrees=45
  swing1Motion=Free
  swing2LimitDegrees=45
  swing2Motion=Free
  twistLimitDegrees=45
  twistMotion=Free
 }
 {bone1Name=dyn_jacket_bk_r_01, bone2Name=dyn_jacket_bk_r_MSTR, swing1LimitDegrees=20, swing2LimitDegrees=30, twistLimitDegrees=5}
 {
  bone1Name=dyn_jacket_r_MSTR
  bone2Name=spine_03
  swing1LimitDegrees=45
  swing1Motion=Free
  swing2LimitDegrees=45
  swing2Motion=Free
  twistLimitDegrees=45
  twistMotion=Free
 }
 {bone1Name=dyn_jacket_r_01, bone2Name=dyn_jacket_r_MSTR, swing1LimitDegrees=45, swing2LimitDegrees=45, twistLimitDegrees=5}
 {
  bone1Name=dyn_jacket_fr_r_MSTR
  bone2Name=spine_03
  swing1LimitDegrees=45
  swing1Motion=Free
  swing2LimitDegrees=45
  swing2Motion=Free
  twistLimitDegrees=45
  twistMotion=Free
 }
 {bone1Name=dyn_jacket_fr_r_01, bone2Name=dyn_jacket_fr_r_MSTR, swing1LimitDegrees=45, swing2LimitDegrees=45, twistLimitDegrees=5}
 {bone1Name=dyn_jacket_fr_r_02, bone2Name=dyn_jacket_fr_r_01, swing1LimitDegrees=45, swing2LimitDegrees=45, twistLimitDegrees=5}
 {
  bone1Name=head
  bone2Name=spine_05
  swing1LimitDegrees=45
  swing1Motion=Free
  swing2LimitDegrees=45
  swing2Motion=Free
  twistLimitDegrees=45
  twistMotion=Free
 }
 {
  bone1Name=spine_05
  bone2Name=spine_03
  swing1LimitDegrees=45
  swing1Motion=Free
  swing2LimitDegrees=45
  swing2Motion=Free
  twistLimitDegrees=45
  twistMotion=Free
 }
 {
  bone1Name=dyn_collar_clasp
  bone2Name=spine_05
  swing1LimitDegrees=45
  swing1Motion=Locked
  swing2LimitDegrees=45
  twistLimitDegrees=45
  twistMotion=Locked
 }
]
